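# Idle Niagara effect for the Ichor Incisor pickaxe, emitting looping, twinkling cyan-to-white star sprites.
# GPU emitter Stars places particles on the pickaxe FX static mesh, with StaticMeshLocation running in particle update rather than spawn.
# Star_LowAndMed, named as a lower-quality variant, omits mesh sampling, spawning fewer sprites in a stretched sphere without GPU simulation.
# Both share curve-driven brightness, alpha and size, plus rotation, drag, curl noise and camera offset.
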NiagaraSystem /BRCosmetics/Weapons/FORT_Melee/Pickaxe_IchorIncisor/FX/NS_Pickaxe_IchorIncisor_Idle
dependencies {
 usedDataInterfaces [&NiagaraDataInterfaceStaticMesh, &NiagaraDataInterfaceCurve]
 usedDynamicInputs []
 usedModules [&SystemState]
}
emitters {
 Star_LowAndMed {
  emitterUpdateScript [
   &EmitterState(Life Cycle Mode=::NewEnumerator1, Loop Duration=1, UseLoopDelay=false)
   &SpawnRate(SpawnRate=&RandomRangeFloat(Minimum=17, Maximum=30), Use Spawn Probability=false)
  ]
  particleSpawnScript [
   &InitializeParticle(
    Color Maximum=<rgba 1 1 1 0.1>
    Color Minimum=<rgba 0.3034 0.791 0.9697 0.5>
    Color Mode=::NewEnumerator2
    Lifetime=&RandomRangeFloat(Minimum=0.8, Maximum=1.3)
    Position Offset=<-7 0 0>
    Sprite Rotation Mode=::NewEnumerator2
    Sprite Size Mode=::NewEnumerator3
    Uniform Sprite Size=&Multiply_Float(A=&RandomRangeFloat(Minimum=8, Maximum=11), B=0.7)
    UsePositionOffset=true
    Write Material Random=true
   ) {module=InitializeParticle001}
   &ShapeLocation(
    MaskSpawnGroup=false
    Non Uniform Scale=<0.2 0.2 4>
    Offset=<8 0 53>
    Offset Mode=::NewEnumerator0
    Override Seed=false
    Sphere Radius=10
    Sphere Surface Distribution=1
    UseOwnerScale=true
   )
  ]
  particleUpdateScript [
   &ParticleState()
   &ScaleColor(
    Scale Alpha=&FloatFromCurve(FloatCurve=DataInterface() {propertyValues {Curve {keys [
         <at 0 0 RCIM_Cubic>
         {arriveTangent=1.1875, interpMode=RCIM_Cubic, leaveTangent=1.1875, time=0.1, value=1.2}
         {arriveTangent=-1.3295, interpMode=RCIM_Cubic, leaveTangent=-1.3295, time=0.8, value=0.95}
         <at 1.0011 0.002 RCIM_Cubic>
        ]}}})
    Scale RGB=&VectorFromFloat(Value=&FloatFromCurve(
      FloatCurve=DataInterface() {propertyValues {Curve {keys [
          <at 0 1 RCIM_Cubic>
          {arriveTangent=-2.0833, interpMode=RCIM_Cubic, leaveTangent=-2.0833, time=0.1, value=1.375}
          {arriveTangent=-1.0417, interpMode=RCIM_Cubic, leaveTangent=-1.0417, time=0.3, value=0.375}
          {arriveTangent=-0.5357, interpMode=RCIM_Cubic, leaveTangent=-0.5357, time=0.7, value=0.75}
          <at 1 0 RCIM_Cubic>
         ]}}}
      Scale Curve=12
     ))
    ScaleA=true
    ScaleRGB=true
   )
   &SpriteRotationRate(Rotation Rate=2)
   &Drag(Drag=4, Use Linear Drag=true, Use Rotational Drag=false)
   &CurlNoiseForce(Noise Frequency=250, Noise Quality / Cost=::NewEnumerator0, Noise Strength=9, Pan Noise=false)
   &SolveForcesAndVelocity(Clamp Velocity=false, Limit Acceleration=false)
   &ScaleSpriteSize(Uniform Curve Sprite Scale=DataInterface() {propertyValues {Curve {keys [
        <at 0 0 RCIM_Cubic>
        {arriveTangent=0.2143, interpMode=RCIM_Cubic, leaveTangent=0.2143, time=0.2, value=0.45}
        {arriveTangent=-0.1875, interpMode=RCIM_Cubic, leaveTangent=-0.1875, time=0.7, value=0.15}
        <at 1 0.3 RCIM_Cubic>
       ]}}})
   &CameraOffset(Camera Offset Amount=-20)
  ]
  rendererClasses [&NiagaraSpriteRendererProperties]
  renderers [{rendererClass=&NiagaraSpriteRendererProperties}]
 }
 Stars {
  simTarget=GPUComputeSim
  emitterUpdateScript [
   &EmitterState(Life Cycle Mode=::NewEnumerator1, Loop Duration=1, UseLoopDelay=false)
   &SpawnRate(SpawnRate=&RandomRangeFloat(Minimum=22, Maximum=40), Use Spawn Probability=false)
  ]
  particleSpawnScript [&InitializeParticle(
    Color Maximum=<rgba 1 1 1 0.1>
    Color Minimum=<rgba 0.3034 0.791 0.9697 0.5>
    Color Mode=::NewEnumerator2
    Lifetime=&RandomRangeFloat(Minimum=0.8, Maximum=1.3)
    Position Offset=<-7 0 0>
    Sprite Rotation Mode=::NewEnumerator2
    Sprite Size Mode=::NewEnumerator3
    Uniform Sprite Size=&Multiply_Float(A=&RandomRangeFloat(Minimum=8, Maximum=11), B=0.7)
    UsePositionOffset=true
    Write Material Random=true
   ) {module=InitializeParticle001}]
  particleUpdateScript [
   &ParticleState()
   /Niagara/Modules/Spawn/Location/StaticMeshLocation(
    MaskSpawnGroup=false
    OffsetAlongNormal=false
    OverrideRandomSeed=false
    SamplePosOffset=false
    Static Mesh=DataInterface() {propertyValues {
      DefaultMesh=&/SM_Pickaxe_IchorIncisor_FX
      InstanceIndex=-1
      LODIndex=0
      PreviewMesh=&SM_Pickaxe_IchorIncisor_FX
      SoftSourceActor=null
      SourceMode=Default
      bAllowSamplingFromStreamingLODs=false
      bCaptureTransformsPerFrame=true
      bUsePhysicsBodyVelocity=false
      FilteredSockets []
      LODIndexUserParameter {parameter {name=None, typeDefHandle {registeredTypeIndex=108}, varData []}}
      MeshParameterBinding {parameter {name=None, typeDefHandle {registeredTypeIndex=1}, varData []}}
      SectionFilter {allowedMaterialSlots []}
     }}
   )
   &ScaleColor(
    Scale Alpha=&FloatFromCurve(FloatCurve=DataInterface() {propertyValues {Curve {keys [
         <at 0 0 RCIM_Cubic>
         {arriveTangent=1.1875, interpMode=RCIM_Cubic, leaveTangent=1.1875, time=0.1, value=1.2}
         {arriveTangent=-1.3295, interpMode=RCIM_Cubic, leaveTangent=-1.3295, time=0.8, value=0.95}
         <at 1.0011 0.002 RCIM_Cubic>
        ]}}})
    Scale RGB=&VectorFromFloat(Value=&FloatFromCurve(
      FloatCurve=DataInterface() {propertyValues {Curve {keys [
          <at 0 1 RCIM_Cubic>
          {arriveTangent=-2.0833, interpMode=RCIM_Cubic, leaveTangent=-2.0833, time=0.1, value=1.375}
          {arriveTangent=-1.0417, interpMode=RCIM_Cubic, leaveTangent=-1.0417, time=0.3, value=0.375}
          {arriveTangent=-0.5357, interpMode=RCIM_Cubic, leaveTangent=-0.5357, time=0.7, value=0.75}
          <at 1 0 RCIM_Cubic>
         ]}}}
      Scale Curve=12
     ))
    ScaleA=true
    ScaleRGB=true
   )
   &SpriteRotationRate(Rotation Rate=2)
   &Drag(Drag=4, Use Linear Drag=true, Use Rotational Drag=false)
   &CurlNoiseForce(Noise Frequency=250, Noise Quality / Cost=::NewEnumerator0, Noise Strength=9, Pan Noise=false)
   &SolveForcesAndVelocity(Clamp Velocity=false, Limit Acceleration=false)
   &ScaleSpriteSize(Uniform Curve Sprite Scale=DataInterface() {propertyValues {Curve {keys [
        <at 0 0 RCIM_Cubic>
        {arriveTangent=0.2143, interpMode=RCIM_Cubic, leaveTangent=0.2143, time=0.2, value=0.45}
        {arriveTangent=-0.1875, interpMode=RCIM_Cubic, leaveTangent=-0.1875, time=0.7, value=0.15}
        <at 1 0.3 RCIM_Cubic>
       ]}}})
   &CameraOffset(Camera Offset Amount=-20)
  ]
  rendererClasses [&NiagaraSpriteRendererProperties]
  renderers [{rendererClass=&NiagaraSpriteRendererProperties}]
 }
}
summary {
 systemName=NS_Pickaxe_IchorIncisor_Idle
 emitters [
  {emitterName=Stars, simTarget=GPUComputeSim, rendererClasses [&NiagaraSpriteRendererProperties]}
  {emitterName=Star_LowAndMed, rendererClasses [&NiagaraSpriteRendererProperties]}
 ]
 userVariables []
}
